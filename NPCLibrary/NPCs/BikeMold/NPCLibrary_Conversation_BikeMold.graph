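# Conversation database for the BikeMold NPC, who speaks as Iron Man and greets players who start a conversation.
# A single entry leads to a speech node that has a default line and contextual lines for first-ever meetings, specific player outfits and match conversation count, with priority curves on the first two kinds.
# The speech node branches through link nodes to shared NPC quests, basic services, BikeMold dialogue and two Season 31 conversation entries.
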ConversationDatabase /NPCLibrary/NPCs/BikeMold/NPCLibrary_Conversation_BikeMold
entry_points [{tag=Conversation.Entry.BikeMold, destinations [#1]}]
nodes [
 ConversationEntryPointNode_7 {
  id=#1
  class=ConversationEntryPointNode
  outputs [#2]
  properties {entryTag {tagName=Conversation.Entry.BikeMold}}
 }
 Speech_C_4 {
  id=#2
  class=Speech_C
  outputs [
   #3
   #4
   #5
   #6
   #7
  ]
  properties {
   nodeName=""
   generalConfig {
    defaultMessage=Is it a job for Iron Man?
    contextualMessages [
     {
      message=I am Iron Man. Obviously.
      contextRequirements [{requirement {class=FortControllerRequirement_IsFirstEverConversationWithNPC}}]
      priority {curve {rowName=Tandem.Speech.Priority.FirstEverInteraction}}
     }
     {
      message="Rhodey! What did you do to your suit?"
      contextRequirements [{requirement {class=FortControllerRequirement_HasCID, properties {allowedCIds [/BRCosmetics/Athena/Items/Cosmetics/Characters/Character_MoosePorch_Math]}}}]
      priority {curve {rowName=Tandem.Speech.Priority.OutfitReaction}}
     }
     {
      message=Nice shield. Really adds to the whole ensemble.
      contextRequirements [{requirement {class=FortControllerRequirement_HasCID, properties {allowedCIds [/BRCosmetics/Athena/Items/Cosmetics/Characters/Character_HerbHutch_Breath]}}}]
      priority {curve {rowName=Tandem.Speech.Priority.OutfitReaction}}
     }
     {
      message=The secret to being a hero? Your brain's your best weapon.
      contextRequirements [{requirement {
         class=FortControllerRequirement_HasCID
         properties {allowedCIds [
           &Character_VoidRedemption_Rebel
           &Character_BrightDisk
          ]}
        }}]
      priority {curve {rowName=Tandem.Speech.Priority.OutfitReaction}}
     }
     {
      message="I've already told you, Shuri, I can upgrade my own suit."
      contextRequirements [{requirement {class=FortControllerRequirement_HasCID, properties {allowedCIds [/BRCosmetics/Athena/Items/Cosmetics/Characters/Character_SnapFreeze_Hunt]}}}]
      priority {curve {rowName=Tandem.Speech.Priority.OutfitReaction}}
     }
     {
      message="I don't care how many walls there are, I can break through anything."
      contextRequirements [{requirement {
         class=FortControllerRequirement_HasCID
         properties {allowedCIds [/BRCosmetics/Athena/Items/Cosmetics/Characters/Character_PillowMill_Mastery]}
        }}]
      priority {curve {rowName=Tandem.Speech.Priority.OutfitReaction}}
     }
     {
      message=So your skeleton is wrapped in... nanamantium? What's that like?
      contextRequirements [{requirement {class=FortControllerRequirement_HasCID, properties {allowedCIds [/BRCosmetics/Athena/Items/Cosmetics/Characters/Character_TroutWrist_Spine]}}}]
      priority {curve {rowName=Tandem.Speech.Priority.OutfitReaction}}
     }
     {
      message=Is it a job for Iron Man?
      contextRequirements [{requirement {class=FortControllerRequirement_CheckNPCMatchConversationCount, properties {nPCUniqueIdTag {tagName=AISpawnerData.Type.Tandem.BikeMold}}}}]
      priority {curve {rowName=None}}
     }
     {
      message="Find any of my tech? Helpful, right?"
      contextRequirements [{requirement {class=FortControllerRequirement_CheckNPCMatchConversationCount, properties {nPCUniqueIdTag {tagName=AISpawnerData.Type.Tandem.BikeMold}}}}]
      priority {curve {rowName=None}}
     }
     {
      message="What, you've never seen an Iron Man before?"
      contextRequirements [{requirement {class=FortControllerRequirement_CheckNPCMatchConversationCount, properties {nPCUniqueIdTag {tagName=AISpawnerData.Type.Tandem.BikeMold}}}}]
      priority {curve {rowName=None}}
     }
    ]
   }
  }
 }
 ConversationLinkNode_4 {
  id=#7
  class=ConversationLinkNode
  properties {remoteEntryTag {tagName=Conversation.Entry.Season31.Found.GreenTown.Q02b}}
 }
 ConversationLinkNode_3 {id=#6, class=ConversationLinkNode, properties {remoteEntryTag {tagName=Conversation.Entry.BikeMold.Dialogue}}}
 ConversationLinkNode_2 {
  id=#5
  class=ConversationLinkNode
  properties {remoteEntryTag {tagName=Conversation.Entry.Season31.Story.P05.Q01}}
 }
 ConversationLinkNode_0 {id=#4, class=ConversationLinkNode, properties {remoteEntryTag {tagName=Conversation.Entry.BasicServices}}}
 ConversationLinkNode_1 {id=#3, class=ConversationLinkNode, properties {remoteEntryTag {tagName=Conversation.Entry.NPCLibrary.Quests}}}
]
