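# Conversation database for the BikeMold NPC, an Iron Man character, providing his "Ask about..." dialogue on the Conversation.Entry.BikeMold.Dialogue entry tag.
# The entry speech branches into choices for Iron Man, Doom, or a Back node.
# Each topic plays a linear chain of four Speech nodes, with Iron Man joking about his name or remarking on Doom and Latveria.
# Both chains converge on a shared ReturnToStart node labeled BACK.
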ConversationDatabase /NPCLibrary/NPCs/BikeMold/NPCLibrary_Conversation_BikeMold_Dialogue
entry_points [{tag=Conversation.Entry.BikeMold.Dialogue, destinations [#1]}]
nodes [
 ConversationEntryPointNode_0 {
  id=#1
  class=ConversationEntryPointNode
  outputs [#2]
  properties {entryTag {tagName=Conversation.Entry.BikeMold.Dialogue}}
 }
 Speech_C_0 {
  id=#2
  class=Speech_C
  outputs [#3, #4, #5]
  properties {
   nodeName=""
   softTaskIcon=&T_UI_NPC_question
   generalConfig {defaultMessage=A question? Excellent. I'm my favorite conversation topic.}
  }
  sub_nodes [ConversationChoiceNode_5 {class=ConversationChoiceNode, properties {defaultChoiceDisplayText=Ask about...}}]
 }
 Speech_C_2 {
  id=#5
  class=Speech_C
  outputs [#6]
  properties {
   nodeName=""
   softTaskIcon=&T_UI_NPC_question
   generalConfig {defaultMessage=Doom's always had some scheme or another up his sleeve.}
  }
  sub_nodes [ConversationChoiceNode_15 {class=ConversationChoiceNode, properties {defaultChoiceDisplayText=Doom}}]
 }
 Speech_C_3 {
  id=#6
  class=Speech_C
  outputs [#7]
  properties {
   nodeName=""
   softTaskIcon=&T_UI_NPC_question
   generalConfig {defaultMessage=Doesn't really surprise me he wants to turn your island into his own little Latveria.}
  }
 }
 Speech_C_4 {
  id=#7
  class=Speech_C
  outputs [#8]
  properties {
   nodeName=""
   softTaskIcon=&T_UI_NPC_question
   generalConfig {defaultMessage="Say what you want, but he keeps a tight ship. Latveria's clean as a whistle."}
  }
 }
 Speech_C_5 {
  id=#8
  class=Speech_C
  outputs [#9]
  properties {
   nodeName=""
   softTaskIcon=&T_UI_NPC_question
   generalConfig {defaultMessage=So long as you don't mind the whole 24/7 surveillance thing.}
  }
 }
 ReturnToStart_C_23 {
  id=#9
  class=ReturnToStart_C
  properties {softTaskIcon=&T_UI_NPC_Back}
  sub_nodes [ConversationChoiceNode_32 {class=ConversationChoiceNode, properties {defaultChoiceDisplayText=BACK}}]
 }
 Back_C_0 {
  id=#4
  class=Back_C
  sub_nodes [ConversationChoiceNode_7 {class=ConversationChoiceNode, properties {defaultChoiceDisplayText=BACK}}]
 }
 Speech_C_7 {
  id=#3
  class=Speech_C
  outputs [#10]
  properties {nodeName="", softTaskIcon=&T_UI_NPC_question, generalConfig {defaultMessage="Yep, that's me. I am Iron Man."}}
  sub_nodes [ConversationChoiceNode_0 {class=ConversationChoiceNode, properties {defaultChoiceDisplayText=Iron Man}}]
 }
 Speech_C_15 {
  id=#10
  class=Speech_C
  outputs [#11]
  properties {
   nodeName=""
   softTaskIcon=&T_UI_NPC_question
   generalConfig {defaultMessage="Though, really it's a titanium-gold alloy. Iron just wasn't cutting it strength-wise."}
  }
 }
 Speech_C_14 {
  id=#11
  class=Speech_C
  outputs [#12]
  properties {
   nodeName=""
   softTaskIcon=&T_UI_NPC_question
   generalConfig {defaultMessage="Still, it's got a nice ring to it, don't you think?"}
  }
 }
 Speech_C_1 {
  id=#12
  class=Speech_C
  outputs [#9]
  properties {
   nodeName=""
   softTaskIcon=&T_UI_NPC_question
   generalConfig {defaultMessage="The name, I mean. Iron itself actually makes a sort of dull \"clunk\" sound."}
  }
 }
]
